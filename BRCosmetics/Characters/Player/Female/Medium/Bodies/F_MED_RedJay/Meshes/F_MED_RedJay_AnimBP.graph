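# Animation Blueprint for the F_MED_RedJay female player body, layering a Control Rig and secondary motion onto the pose copied from the parent mesh.
# Cached poses chain the rig into additive bone tweaks, then Anim Dynamics springs on the gun and holder bones, with a spherical limit on the left thigh.
# The update event sets FrontEndPosing when in the front end without skydiving or emoting, though the shown AnimGraph never reads it.
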AnimBlueprint /BRCosmetics/Characters/Player/Female/Medium/Bodies/F_MED_RedJay/Meshes/F_MED_RedJay_AnimBP
target_skeleton=&/F_MED_RedJay_Skeleton
graphs [
 AnimGraph {nodes [
   Output Pose(Result=Use cached pose 'CRCache_dynamics'()) {layerGroup=None}
   n1 = Control Rig!(Source=Copy Pose From Mesh() {bUseAttachedParent=true}, Alpha, Is Crouching:bool=false) -> (Pose) {
    alphaBoolBlend {blendOption=Linear}
    controlRigAssetReference {blueprintRigClass=&/F_MED_RedJay_CtrlRig.F_MED_RedJay_CtrlRig_C}
   }
   Save cached pose 'CRCache_inputPose'(Pose=@n1) {cacheName=CRCache_inputPose}
   n2 = "Transform (Modify) Bone - Bone: attach_backpack"!(ComponentPose=Local To Component(LocalPose=Use cached pose 'CRCache_inputPose'()), Alpha=1, Translation="0,0,0", Rotation="0,0,0") -> (Pose) {rotationMode=BMM_Additive, translationMode=BMM_Additive, alphaBoolBlend {blendOption=Linear}, boneToModify {boneName=attach_backpack}}
   n3 = "Transform (Modify) Bone - Bone: dyn_gunHolder_02"!(ComponentPose=@n2, bAlphaBoolEnabled:bool=True, Translation="2,0,0") -> (Pose) {
    alphaInputType=Bool
    translationMode=BMM_Additive
    translationSpace=BCS_BoneSpace
    alphaBoolBlend {blendInTime=0.1, blendOption=Linear, blendOutTime=0.1}
    boneToModify {boneName=dyn_gunHolder_02}
    translation {x=2}
   }
   Save cached pose 'CRCache_preDynamics'(Pose=Component To Local(ComponentPose=@n3)) {cacheName=CRCache_preDynamics}
   n4 = Anim Dynamics - Bone: dyn_gun(ComponentPose=Local To Component(LocalPose=Use cached pose 'CRCache_preDynamics'()), Alpha=1, ExternalForce="0,0,0") {
    angularDampingOverride=1
    angularSpringConstant=5000
    bAngularSpring=true
    bOverrideAngularDamping=true
    bUseGravityOverride=true
    gravityScale=0.35
    numSolverIterationsPreUpdate=10
    simulationSpace=RootRelative
    alphaBoolBlend {blendOption=Linear}
    boundBone {boneName=dyn_gun}
    componentAppliedLinearAccClamp=<1000 1000 1000>
    componentLinearAccScale=<1 1 1>
    componentLinearVelScale=<1 1 1>
    gravityOverride {z=-300}
    physicsBodyDefinitions [{
      boundBone {boneName=dyn_gun}
      boxExtents=<25 30 30>
      constraintSetup {angularLimitsMax=<0 0 1>, angularLimitsMin=<0 0 -2>, angularTarget=<1 0 0>}
      localJointOffset=<1 0 0>
     }]
   }
   n5 = Anim Dynamics - Bone: dyn_gunHolder_01(ComponentPose=@n4, Alpha=1, ExternalForce="0,0,0") {
    angularDampingOverride=1
    angularSpringConstant=20000
    bAngularSpring=true
    bOverrideAngularDamping=true
    bUseGravityOverride=true
    bUseSphericalLimits=true
    gravityScale=0.35
    simulationSpace=RootRelative
    alphaBoolBlend {blendOption=Linear}
    boundBone {boneName=dyn_gunHolder_01}
    componentLinearAccScale=<1 1 1>
    componentLinearVelScale=<1 1 1>
    gravityOverride {z=-980}
    physicsBodyDefinitions [{
      boundBone {boneName=dyn_gunHolder_01}
      boxExtents=<20 30 30>
      constraintSetup {angularLimitsMax=<0 4 5>, angularLimitsMin=<0 -3 -2>, angularTarget=<1 0 0>}
      localJointOffset=<2 0 0>
     }]
    sphericalLimits [{limitRadius=10, drivingBone {boneName=thigh_l}, sphereLocalOffset=<-10 0 3>}]
   }
   Save cached pose 'CRCache_dynamics'(Pose=Component To Local(ComponentPose=@n5)) {cacheName=CRCache_dynamics}
  ]}
 EventGraph {nodes [
   Event BlueprintUpdateAnimation()
   n1 = Get bIsSkyDiving()
   n2 = Get bIsPlayingEmote()
   > n3 = Branch(Condition=AND Boolean!(A=Get bIsInFrontEnd(), B=NOT Boolean(A=@n1), C:bool=NOT Boolean(A=@n2)) -> (ReturnValue))
   > Set FrontEndPosing(FrontEndPosing:bool=true) -> (Output_Get:bool)
   Set FrontEndPosing(execute=@n3.else, FrontEndPosing=false)
  ]}
]
variables [FrontEndPosing {type=boolean}]
