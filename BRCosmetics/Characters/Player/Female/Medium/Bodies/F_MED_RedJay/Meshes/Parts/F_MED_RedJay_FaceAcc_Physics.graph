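# Physics asset for the RedJay medium female body's face accessory mesh, simulating four front hair strands (A, B, L, R).
# Each strand is a chain of two or three small capsule bodies hanging from a base body, whose constraint to the head is angularly free.
# Chain joints have narrow swing limits and zero twist, with A and B confined to one swing axis below their first joint.
# Three capsules on the head body approximate the head volume.
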PhysicsAsset /BRCosmetics/Characters/Player/Female/Medium/Bodies/F_MED_RedJay/Meshes/Parts/F_MED_RedJay_FaceAcc_Physics
bodies {
 base_front_hair_A [{shapeType=Capsule, length=0.505, radius=0.505, shapeName=base_front_hair_A_capsule, rotation=<rot 0 -90 0>}]
 base_front_hair_B [{shapeType=Capsule, length=0.505, radius=0.505, shapeName=base_front_hair_B_capsule, rotation=<rot 0 -90 0>}]
 base_front_hair_L [{shapeType=Capsule, length=0.505, radius=0.505, shapeName=base_front_hair_L_capsule, rotation=<rot 0 -90 0>}]
 base_front_hair_R [{shapeType=Capsule, length=0.505, radius=0.505, shapeName=base_front_hair_R_capsule, rotation=<rot 0 -90 0>}]
 dyn_front_hair_A_01 [{shapeType=Capsule, length=1, radius=1, shapeName=dyn_front_hair_A_01_capsule, center=<2 0 0>, rotation=<rot 90 0 0>}]
 dyn_front_hair_A_02 [{shapeType=Capsule, length=1, radius=0.6, shapeName=dyn_front_hair_A_02_capsule, center=<2 0 0>, rotation=<rot 90 0 0>}]
 dyn_front_hair_A_03 [{
   shapeType=Capsule
   length=0.3
   radius=0.5
   shapeName=dyn_front_hair_A_03_capsule
   center=<0.969 0 0.2694>
   rotation=<rot 61.9077 -179.9999 -180>
  }]
 dyn_front_hair_B_01 [{shapeType=Capsule, length=1, radius=1, shapeName=dyn_front_hair_B_01_capsule, center=<2 0 0>, rotation=<rot 90 0 0>}]
 dyn_front_hair_B_02 [{shapeType=Capsule, length=1, radius=1, shapeName=dyn_front_hair_B_02_capsule, center=<2 0 0>, rotation=<rot 90 0 0>}]
 dyn_front_hair_B_03 [{
   shapeType=Capsule
   length=0.2
   radius=0.5
   shapeName=dyn_front_hair_B_03_capsule
   center=<1.1658 0 0.5059>
   rotation=<rot 78.8208 -179.9972 -179.9977>
  }]
 dyn_front_hair_L_01 [{shapeType=Capsule, length=1, radius=1, shapeName=dyn_front_hair_L_01_capsule, center=<2 0 0>, rotation=<rot 90 0 0>}]
 dyn_front_hair_L_02 [{shapeType=Capsule, length=1, radius=1.3478, shapeName=dyn_front_hair_L_02_capsule, center=<2 0 0>, rotation=<rot 90 9.9465 0>}]
 dyn_front_hair_R_01 [{shapeType=Capsule, length=1, radius=1, shapeName=dyn_front_hair_R_01_capsule, center=<2 0 0>, rotation=<rot 90 0 0>}]
 dyn_front_hair_R_02 [{
   shapeType=Capsule
   length=3.273
   radius=1
   shapeName=dyn_front_hair_R_02_capsule
   center=<2.0603 0 0.5983>
   rotation=<rot 80.0007 -0.0001 -0.0002>
  }]
 dyn_front_hair_R_03 [{shapeType=Capsule, length=1, radius=1.3422, shapeName=dyn_front_hair_R_03_capsule, center=<2 0 0>, rotation=<rot 90 0 0>}]
 head [
  {
   shapeType=Capsule
   length=5.5327
   radius=6.1821
   shapeName=head_capsule
   center=<3.6577 2.1402 -0.0367>
   rotation=<rot -8.1579 -83.2828 -91.4572>
  }
  {shapeType=Capsule, length=3.1735, radius=1.634, shapeName=head_capsule, center=<2.842 9.002 -3.3991>, rotation=<rot 12.3709 2.617 12.044>}
  {
   shapeType=Capsule
   length=4.4973
   radius=1.8016
   shapeName=head_capsule
   center=<2.6298 9.2076 3.3373>
   rotation=<rot -14.3523 -0.0001 -8.8847>
  }
 ]
}
constraints [
 {
  bone1Name=base_front_hair_A
  bone2Name=head
  swing1LimitDegrees=45
  swing1Motion=Free
  swing2LimitDegrees=45
  swing2Motion=Free
  twistLimitDegrees=45
  twistMotion=Free
 }
 {
  bone1Name=base_front_hair_B
  bone2Name=head
  swing1LimitDegrees=45
  swing1Motion=Free
  swing2LimitDegrees=45
  swing2Motion=Free
  twistLimitDegrees=45
  twistMotion=Free
 }
 {
  bone1Name=base_front_hair_L
  bone2Name=head
  swing1LimitDegrees=45
  swing1Motion=Free
  swing2LimitDegrees=45
  swing2Motion=Free
  twistLimitDegrees=45
  twistMotion=Free
 }
 {
  bone1Name=base_front_hair_R
  bone2Name=head
  swing1LimitDegrees=45
  swing1Motion=Free
  swing2LimitDegrees=45
  swing2Motion=Free
  twistLimitDegrees=45
  twistMotion=Free
 }
 {bone1Name=dyn_front_hair_A_01, bone2Name=base_front_hair_A, swing1LimitDegrees=25, swing2LimitDegrees=5, twistLimitDegrees=0}
 {bone1Name=dyn_front_hair_A_03, bone2Name=dyn_front_hair_A_02, swing1LimitDegrees=5, swing2LimitDegrees=0, twistLimitDegrees=0}
 {bone1Name=dyn_front_hair_A_02, bone2Name=dyn_front_hair_A_01, swing1LimitDegrees=20, swing2LimitDegrees=0, twistLimitDegrees=0}
 {bone1Name=dyn_front_hair_B_01, bone2Name=base_front_hair_B, swing1LimitDegrees=25, swing2LimitDegrees=5, twistLimitDegrees=0}
 {bone1Name=dyn_front_hair_B_02, bone2Name=dyn_front_hair_B_01, swing1LimitDegrees=20, swing2LimitDegrees=0, twistLimitDegrees=0}
 {bone1Name=dyn_front_hair_B_03, bone2Name=dyn_front_hair_B_02, swing1LimitDegrees=5, swing2LimitDegrees=0, twistLimitDegrees=0}
 {bone1Name=dyn_front_hair_L_01, bone2Name=base_front_hair_L, swing1LimitDegrees=10, swing2LimitDegrees=10, twistLimitDegrees=0}
 {bone1Name=dyn_front_hair_L_02, bone2Name=dyn_front_hair_L_01, swing1LimitDegrees=10, swing2LimitDegrees=10, twistLimitDegrees=0}
 {bone1Name=dyn_front_hair_R_01, bone2Name=base_front_hair_R, swing1LimitDegrees=10, swing2LimitDegrees=10, twistLimitDegrees=0}
 {bone1Name=dyn_front_hair_R_02, bone2Name=dyn_front_hair_R_01, swing1LimitDegrees=15, swing2LimitDegrees=15, twistLimitDegrees=0}
 {bone1Name=dyn_front_hair_R_03, bone2Name=dyn_front_hair_R_02, swing1LimitDegrees=15, swing2LimitDegrees=15, twistLimitDegrees=0}
]
