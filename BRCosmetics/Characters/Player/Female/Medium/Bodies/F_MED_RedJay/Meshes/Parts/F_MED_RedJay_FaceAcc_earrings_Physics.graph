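# Physics asset for the RedJay female character's earrings face-accessory mesh, with capsule bodies for the head, two-segment earrings and two front hair bones.
# Each upper earring segment hangs from the head with limited swing and twist.
# Each lower segment links to the upper one with almost no swing but a wide twist range.
# Both front hair bodies attach to the head with swing and twist left free.
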PhysicsAsset /BRCosmetics/Characters/Player/Female/Medium/Bodies/F_MED_RedJay/Meshes/Parts/F_MED_RedJay_FaceAcc_earrings_Physics
bodies {
 dyn_front_hair_L_02 [{
   shapeType=Capsule
   length=4.4364
   radius=0.6863
   shapeName=dyn_front_hair_L_02_capsule
   center=<1.8266 0.9234 1.3987>
   rotation=<rot 2.7484 110.6883 -87.927>
  }]
 dyn_front_hair_R_02 [{
   shapeType=Capsule
   length=4.4364
   radius=0.6863
   shapeName=dyn_front_hair_R_02_capsule
   center=<7.0965 0.0551 1.0251>
   rotation=<rot 9.5553 77.5172 94.6359>
  }]
 dyn_left_earring_01 [{shapeType=Capsule, length=3.5186, radius=0.505, shapeName=dyn_left_earring_01_capsule, center=<2.5 0 0>, rotation=<rot 90 0 0>}]
 dyn_left_earring_02 [{
   shapeType=Capsule
   length=2.1357
   radius=0.7025
   shapeName=dyn_left_earring_02_capsule
   center=<2 -0.0029 -0.0992>
   rotation=<rot 0.8148 178.1464 -93.6818>
  }]
 dyn_right_earring_01 [{shapeType=Capsule, length=3.5391, radius=0.505, shapeName=dyn_right_earring_01_capsule, center=<-2.5 0 0>, rotation=<rot 90 0 0>}]
 dyn_right_earring_02 [{
   shapeType=Capsule
   length=2.1357
   radius=0.7025
   shapeName=dyn_right_earring_02_capsule
   center=<-2 0.0029 0.0992>
   rotation=<rot 0.8148 178.1464 86.3182>
  }]
 head [{shapeType=Capsule, length=5.7051, radius=4.7799, center=<1.7981 3.144 -0.216>, rotation=<rot 23.24 -122.3146 -90>}]
}
constraints [
 {bone1Name=dyn_left_earring_01, bone2Name=head, swing1LimitDegrees=40, swing2LimitDegrees=20, twistLimitDegrees=45}
 {bone1Name=dyn_right_earring_01, bone2Name=head, swing1LimitDegrees=40, swing2LimitDegrees=20, twistLimitDegrees=45}
 {bone1Name=dyn_left_earring_02, bone2Name=dyn_left_earring_01, swing1LimitDegrees=2, swing2LimitDegrees=2, twistLimitDegrees=90}
 {bone1Name=dyn_right_earring_02, bone2Name=dyn_right_earring_01, swing1LimitDegrees=2, swing2LimitDegrees=2, twistLimitDegrees=90}
 {
  bone1Name=dyn_front_hair_L_02
  bone2Name=head
  swing1LimitDegrees=45
  swing1Motion=Free
  swing2LimitDegrees=45
  swing2Motion=Free
  twistLimitDegrees=45
  twistMotion=Free
 }
 {
  bone1Name=dyn_front_hair_R_02
  bone2Name=head
  swing1LimitDegrees=45
  swing1Motion=Free
  swing2LimitDegrees=45
  swing2Motion=Free
  twistLimitDegrees=45
  twistMotion=Free
 }
]
