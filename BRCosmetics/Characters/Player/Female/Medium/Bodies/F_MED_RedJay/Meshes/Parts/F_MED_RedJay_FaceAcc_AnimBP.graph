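# Post-process style AnimBP for RedJay's face accessory mesh, adding secondary motion to front hair and earrings.
# It copies the attached parent's pose, runs chained AnimDynamics on front-hair bones, then a head-relative RigidBody using the earrings physics asset.
# The event graph updates gravity override and selects front-end or default sim-space variables.
# Several AnimDynamics nodes and another RigidBody are left unconnected to the output.
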AnimBlueprint /BRCosmetics/Characters/Player/Female/Medium/Bodies/F_MED_RedJay/Meshes/Parts/F_MED_RedJay_FaceAcc_AnimBP
target_skeleton=&/F_MED_RedJay_FaceAcc_Skeleton
graphs [
 AnimGraph {nodes [
   Anim Dynamics - Bone: dyn_left_earring_01(Alpha=1, ExternalForce="0,0,0") {
    angularDampingOverride=1
    angularSpringConstant=20000
    bAngularSpring=true
    bOverrideAngularDamping=true
    bUseGravityOverride=true
    gravityScale=0.35
    simulationSpace=RootRelative
    alphaBoolBlend {blendOption=Linear}
    boundBone {boneName=dyn_left_earring_01}
    componentLinearAccScale=<1 1 1>
    componentLinearVelScale=<1 1 1>
    gravityOverride {z=-980}
    physicsBodyDefinitions [{
      boundBone {boneName=dyn_left_earring_01}
      boxExtents=<20 30 30>
      constraintSetup {angularLimitsMax=<15 15 100>, angularLimitsMin=<-15 -15 -15>, angularTarget=<1 0 0>}
      localJointOffset=<4 0 0>
     }]
   }
   Anim Dynamics - Bone: dyn_right_earring_01(Alpha=1, ExternalForce="0,0,0") {
    angularDampingOverride=1
    angularSpringConstant=7000
    bAngularSpring=true
    bOverrideAngularDamping=true
    bUseGravityOverride=true
    gravityScale=0.35
    alphaBoolBlend {blendOption=Linear}
    boundBone {boneName=dyn_right_earring_01}
    componentAppliedLinearAccClamp=<2000 2000 2000>
    componentLinearAccScale=<1 1 1>
    componentLinearVelScale=<1 1 1>
    gravityOverride {z=-980}
    physicsBodyDefinitions [{
      boundBone {boneName=dyn_right_earring_01}
      boxExtents=<20 30 30>
      constraintSetup {angularLimitsMax=<0 40 15>, angularLimitsMin=<0 -40 -100>, angularTarget=<1 0 0>}
      localJointOffset=<1 0 0>
     }]
   }
   Anim Dynamics - Bone: dyn_left_earring_02(Alpha=1, ExternalForce="0,0,0") {
    angularDampingOverride=1
    angularSpringConstant=50000
    bAngularSpring=true
    bOverrideAngularDamping=true
    bUseGravityOverride=true
    gravityScale=0.35
    simulationSpace=RootRelative
    alphaBoolBlend {blendOption=Linear}
    boundBone {boneName=dyn_left_earring_02}
    componentLinearAccScale=<1 1 1>
    componentLinearVelScale=<1 1 1>
    gravityOverride {z=-980}
    physicsBodyDefinitions [{
      boundBone {boneName=dyn_left_earring_02}
      boxExtents=<20 30 30>
      constraintSetup {angularLimitsMax=<60 60 60>, angularLimitsMin=<-60 -60 -60>, angularTarget=<1 0 0>}
      localJointOffset=<2 0 0>
     }]
   }
   n1 = Anim Dynamics - Bone: dyn_front_hair_A_01!(
    ComponentPose
    Alpha=1
    GravityOverride:Vector="0,0,-980"
    SimSpaceSettings:Anim Phys Sim Space Settings Structure=(SimSpaceAngularAlpha=1,MaxAngularVelocity=20,MaxAngularAcceleration=40,ExternalAngularVelocity=(X=0,Y=0,Z=0))
   ) -> (Pose) {
    angularDampingOverride=0.99
    angularSpringConstant=2000
    bAngularSpring=true
    bOverrideAngularDamping=true
    bUseGravityOverride=true
    bUseSphericalLimits=true
    simulationSpace=RootRelative
    alphaBoolBlend {blendOption=Linear}
    boundBone {boneName=dyn_front_hair_A_01}
    componentAppliedLinearAccClamp=<5000 5000 5000>
    componentLinearAccScale=<0.4 0.4 0.4>
    componentLinearVelScale=<0.5 0.5 0.5>
    gravityOverride {z=-980}
    physicsBodyDefinitions [{
      boundBone {boneName=dyn_front_hair_A_01}
      boxExtents=<20 20 20>
      constraintSetup {angularLimitsMax=<0 20 10>, angularLimitsMin=<0 -20 -10>, angularTarget=<1 0 0>}
      localJointOffset=<1 0 0>
     }]
    simSpaceSettings {maxAngularAcceleration=40, maxAngularVelocity=20, simSpaceAngularAlpha=1}
    sphericalLimits [{limitRadius=10, drivingBone {boneName=head}, sphereLocalOffset=<2 0 0>}]
   }
   Anim Dynamics - Bone: dyn_front_hair_B_01!(
    ComponentPose=@n1
    Alpha=1
    GravityOverride:Vector="0,0,-980"
    SimSpaceSettings:Anim Phys Sim Space Settings Structure=(SimSpaceAngularAlpha=1,MaxAngularVelocity=20,MaxAngularAcceleration=40,ExternalAngularVelocity=(X=0,Y=0,Z=0))
   ) -> (Pose) {
    angularDampingOverride=0.99
    angularSpringConstant=2000
    bAngularSpring=true
    bOverrideAngularDamping=true
    bUseGravityOverride=true
    bUseSphericalLimits=true
    simulationSpace=RootRelative
    alphaBoolBlend {blendOption=Linear}
    boundBone {boneName=dyn_front_hair_B_01}
    componentAppliedLinearAccClamp=<5000 5000 5000>
    componentLinearAccScale=<0.4 0.4 0.4>
    componentLinearVelScale=<0.5 0.5 0.5>
    gravityOverride {z=-980}
    physicsBodyDefinitions [{
      boundBone {boneName=dyn_front_hair_B_01}
      boxExtents=<20 20 20>
      constraintSetup {angularLimitsMax=<0 5 20>, angularLimitsMin=<0 -5 -20>, angularTarget=<1 0 0>}
      localJointOffset=<1 0 0>
     }]
    simSpaceSettings {maxAngularAcceleration=40, maxAngularVelocity=20, simSpaceAngularAlpha=1}
    sphericalLimits [{limitRadius=10, drivingBone {boneName=head}, sphereLocalOffset=<2 0 0>}]
   }
   n2 = Anim Dynamics - Bone: dyn_front_hair_A_03!(
    ComponentPose
    Alpha=1
    GravityOverride:Vector="0,0,-980"
    SimSpaceSettings:Anim Phys Sim Space Settings Structure=(SimSpaceAngularAlpha=1,MaxAngularVelocity=20,MaxAngularAcceleration=40,ExternalAngularVelocity=(X=0,Y=0,Z=0))
   ) -> (Pose) {
    angularDampingOverride=0.99
    angularSpringConstant=2000
    bAngularSpring=true
    bOverrideAngularDamping=true
    bUseGravityOverride=true
    bUseSphericalLimits=true
    simulationSpace=RootRelative
    alphaBoolBlend {blendOption=Linear}
    boundBone {boneName=dyn_front_hair_A_03}
    componentAppliedLinearAccClamp=<3000 3000 3000>
    componentLinearAccScale=<0.4 0.4 0.4>
    componentLinearVelScale=<0.6 0.6 0.6>
    gravityOverride {z=-980}
    physicsBodyDefinitions [{
      collisionType=CustomSphere
      sphereCollisionRadius=1.2
      boundBone {boneName=dyn_front_hair_A_03}
      boxExtents=<20 20 20>
      constraintSetup {angularLimitsMax=<0 25 15>, angularLimitsMin=<0 -25 -20>, angularTarget=<1 0 0>}
      localJointOffset=<1 0 0>
     }]
    simSpaceSettings {maxAngularAcceleration=40, maxAngularVelocity=20, simSpaceAngularAlpha=1}
    sphericalLimits [{limitRadius=10, drivingBone {boneName=head}, sphereLocalOffset=<2 0 0>}]
   }
   Anim Dynamics - Bone: dyn_front_hair_B_03!(
    ComponentPose=@n2
    Alpha=1
    GravityOverride:Vector="0,0,-980"
    SimSpaceSettings:Anim Phys Sim Space Settings Structure=(SimSpaceAngularAlpha=1,MaxAngularVelocity=20,MaxAngularAcceleration=40,ExternalAngularVelocity=(X=0,Y=0,Z=0))
   ) -> (Pose) {
    angularDampingOverride=0.99
    angularSpringConstant=2000
    bAngularSpring=true
    bOverrideAngularDamping=true
    bUseGravityOverride=true
    bUseSphericalLimits=true
    simulationSpace=RootRelative
    alphaBoolBlend {blendOption=Linear}
    boundBone {boneName=dyn_front_hair_B_03}
    componentAppliedLinearAccClamp=<3000 3000 3000>
    componentLinearAccScale=<0.4 0.4 0.4>
    componentLinearVelScale=<0.6 0.6 0.6>
    gravityOverride {z=-980}
    physicsBodyDefinitions [{
      collisionType=CustomSphere
      sphereCollisionRadius=1.2
      boundBone {boneName=dyn_front_hair_B_03}
      boxExtents=<20 20 20>
      constraintSetup {angularLimitsMax=<0 5 25>, angularLimitsMin=<0 -5 -25>, angularTarget=<1 0 0>}
      localJointOffset=<1 0 0>
     }]
    simSpaceSettings {maxAngularAcceleration=40, maxAngularVelocity=20, simSpaceAngularAlpha=1}
    sphericalLimits [{limitRadius=10, drivingBone {boneName=head}, sphereLocalOffset=<2 0 0>}]
   }
   n3 = Anim Dynamics - Bone: dyn_front_hair_L_01(Alpha=1, ExternalForce="0,0,0") {
    angularDampingOverride=0.99
    angularSpringConstant=4000
    bAngularSpring=true
    bOverrideAngularDamping=true
    bUseGravityOverride=true
    bUseSphericalLimits=true
    alphaBoolBlend {blendOption=Linear}
    boundBone {boneName=dyn_front_hair_L_01}
    componentAppliedLinearAccClamp=<2000 2000 2000>
    componentLinearAccScale=<1 1 1>
    componentLinearVelScale=<1 1 1>
    gravityOverride {z=-980}
    physicsBodyDefinitions [{
      boundBone {boneName=dyn_front_hair_L_01}
      boxExtents=<20 20 20>
      constraintSetup {angularLimitsMax=<0 5 5>, angularLimitsMin=<0 -15 -15>, angularTarget=<1 0 0>}
      localJointOffset=<0.8 0 0>
     }]
    simSpaceSettings {maxAngularAcceleration=5000, maxAngularVelocity=5000}
    sphericalLimits [{limitRadius=8, drivingBone {boneName=head}, sphereLocalOffset=<2 3 1>}]
   }
   Anim Dynamics - Bone: dyn_front_hair_L_02(ComponentPose=@n3, Alpha=1, ExternalForce="0,0,0") {
    angularDampingOverride=0.99
    angularSpringConstant=4000
    bAngularSpring=true
    bOverrideAngularDamping=true
    bUseGravityOverride=true
    bUseSphericalLimits=true
    alphaBoolBlend {blendOption=Linear}
    boundBone {boneName=dyn_front_hair_L_02}
    componentAppliedLinearAccClamp=<2000 2000 2000>
    componentLinearAccScale=<1 1 1>
    componentLinearVelScale=<1 1 1>
    gravityOverride {z=-980}
    physicsBodyDefinitions [{
      collisionType=CustomSphere
      sphereCollisionRadius=1.2
      boundBone {boneName=dyn_front_hair_L_02}
      boxExtents=<40 40 40>
      constraintSetup {angularLimitsMax=<0 15 15>, angularLimitsMin=<0 -15 -20>, angularTarget=<1 0 0>}
      localJointOffset=<6 0 0>
     }]
    simSpaceSettings {maxAngularAcceleration=5000, maxAngularVelocity=5000}
    sphericalLimits [{limitRadius=8, drivingBone {boneName=head}, sphereLocalOffset=<0 1.5 -1>}]
   }
   Anim Dynamics - Bone: dyn_left_earring_01(Alpha=1, ExternalForce="0,0,0") {
    angularDampingOverride=1
    angularSpringConstant=7000
    bAngularSpring=true
    bOverrideAngularDamping=true
    bUseGravityOverride=true
    gravityScale=0.35
    simulationSpace=RootRelative
    alphaBoolBlend {blendOption=Linear}
    boundBone {boneName=dyn_left_earring_01}
    componentLinearAccScale=<1 1 1>
    componentLinearVelScale=<1 1 1>
    gravityOverride {z=-980}
    physicsBodyDefinitions [{
      boundBone {boneName=dyn_left_earring_01}
      boxExtents=<20 30 30>
      constraintSetup {angularLimitsMax=<20 40 15>, angularLimitsMin=<-20 -40 -100>, angularTarget=<1 0 0>}
      localJointOffset=<2 0 0>
     }]
   }
   Anim Dynamics - Bone: dyn_right_earring_02(Alpha=1, ExternalForce="0,0,0") {
    angularDampingOverride=1
    angularSpringConstant=50000
    bAngularSpring=true
    bOverrideAngularDamping=true
    bUseGravityOverride=true
    gravityScale=0.35
    simulationSpace=RootRelative
    alphaBoolBlend {blendOption=Linear}
    boundBone {boneName=dyn_right_earring_02}
    componentLinearAccScale=<1 1 1>
    componentLinearVelScale=<1 1 1>
    gravityOverride {z=-980}
    physicsBodyDefinitions [{
      boundBone {boneName=dyn_right_earring_02}
      boxExtents=<20 30 30>
      constraintSetup {angularLimitsMax=<60 60 60>, angularLimitsMin=<-60 -60 -60>, angularTarget=<1 0 0>}
      localJointOffset=<2 0 0>
     }]
   }
   n4 = Anim Dynamics - Bone: dyn_front_hair_R_01!(ComponentPose, Alpha=1, GravityOverride:Vector="0,0,-980", ExternalForce="0,0,0") -> (Pose) {
    angularDampingOverride=0.99
    angularSpringConstant=2000
    bAngularSpring=true
    bOverrideAngularDamping=true
    bUseGravityOverride=true
    bUseSphericalLimits=true
    simulationSpace=RootRelative
    alphaBoolBlend {blendOption=Linear}
    boundBone {boneName=dyn_front_hair_R_01}
    componentAppliedLinearAccClamp=<5000 5000 5000>
    componentLinearAccScale=<1 1 1>
    componentLinearVelScale=<1 1 1>
    gravityOverride {z=-980}
    physicsBodyDefinitions [{
      boundBone {boneName=dyn_front_hair_R_01}
      boxExtents=<20 20 20>
      constraintSetup {angularLimitsMax=<0 20 5>, angularLimitsMin=<0 -20 -10>, angularTarget=<1 0 0>}
      localJointOffset=<0.8 0 0>
     }]
    simSpaceSettings {maxAngularAcceleration=5000, maxAngularVelocity=5000}
    sphericalLimits [{limitRadius=10, drivingBone {boneName=head}, sphereLocalOffset=<2 2 -0.5>}]
   }
   n5 = Anim Dynamics - Bone: dyn_front_hair_R_02!(ComponentPose=@n4, Alpha=1, GravityOverride:Vector="0,0,-980", ExternalForce="0,0,0") -> (Pose) {
    angularDampingOverride=0.99
    angularSpringConstant=3000
    bAngularSpring=true
    bOverrideAngularDamping=true
    bUseGravityOverride=true
    bUseSphericalLimits=true
    simulationSpace=RootRelative
    alphaBoolBlend {blendOption=Linear}
    boundBone {boneName=dyn_front_hair_R_02}
    componentAppliedLinearAccClamp=<5000 5000 5000>
    componentLinearAccScale=<1 1 1>
    componentLinearVelScale=<1 1 1>
    gravityOverride {z=-980}
    physicsBodyDefinitions [{
      boundBone {boneName=dyn_front_hair_R_02}
      boxExtents=<20 20 20>
      constraintSetup {angularLimitsMax=<0 5 5>, angularLimitsMin=<0 -5 -15>, angularTarget=<1 0 0>}
      localJointOffset=<0.8 0 0>
     }]
    simSpaceSettings {maxAngularAcceleration=5000, maxAngularVelocity=5000}
    sphericalLimits [{limitRadius=8, drivingBone {boneName=head}, sphereLocalOffset=<2 3 1>}]
   }
   Anim Dynamics - Bone: dyn_front_hair_R_03!(ComponentPose=@n5, Alpha=1, GravityOverride:Vector="0,0,-980", ExternalForce="0,0,0") -> (Pose) {
    angularDampingOverride=0.99
    angularSpringConstant=4000
    bAngularSpring=true
    bOverrideAngularDamping=true
    bUseGravityOverride=true
    bUseSphericalLimits=true
    simulationSpace=RootRelative
    alphaBoolBlend {blendOption=Linear}
    boundBone {boneName=dyn_front_hair_R_03}
    componentAppliedLinearAccClamp=<5000 5000 5000>
    componentLinearAccScale=<1 1 1>
    componentLinearVelScale=<1 1 1>
    gravityOverride {z=-980}
    physicsBodyDefinitions [{
      collisionType=CustomSphere
      sphereCollisionRadius=1.2
      boundBone {boneName=dyn_front_hair_R_03}
      boxExtents=<40 40 40>
      constraintSetup {angularLimitsMax=<0 15 15>, angularLimitsMin=<0 -15 -20>, angularTarget=<1 0 0>}
      localJointOffset=<6 0 0>
     }]
    simSpaceSettings {maxAngularAcceleration=5000, maxAngularVelocity=5000}
    sphericalLimits [{limitRadius=8, drivingBone {boneName=head}, sphereLocalOffset=<0 1.5 -1>}]
   }
   RigidBody() {
    overridePhysicsAsset=&/F_MED_RedJay_FaceAcc_Physics
    simulationSpace=BaseBoneSpace
    alphaBoolBlend {blendOption=Linear}
    baseBoneRef {boneName=head}
    componentAppliedLinearAccClamp=<2000 2000 2000>
    simSpaceSettings {
     dampingAlpha=0
     maxAngularAcceleration=15
     maxAngularVelocity=7
     maxLinearAcceleration=700
     maxLinearVelocity=500
     velocityScaleZ=0.2
     worldAlpha=1
    }
   }
   Output Pose(Result=Use cached pose 'CRCache_dynamics'()) {layerGroup=None}
   Save cached pose 'CRCache_inputPose'(Pose=Copy Pose From Mesh() {bUseAttachedParent=true}) {cacheName=CRCache_inputPose}
   n6 = Anim Dynamics - Bone: dyn_front_hair_L_01!(
    ComponentPose=Local To Component(LocalPose=Use cached pose 'CRCache_inputPose'())
    Alpha=1
    GravityOverride:Vector="0,0,-980"
    SimSpaceSettings:Anim Phys Sim Space Settings Structure=(SimSpaceAngularAlpha=1,MaxAngularVelocity=20,MaxAngularAcceleration=40,ExternalAngularVelocity=(X=0,Y=0,Z=0))
   ) -> (Pose) {
    angularDampingOverride=0.99
    angularSpringConstant=2000
    bAngularSpring=true
    bOverrideAngularDamping=true
    bUseGravityOverride=true
    bUseSphericalLimits=true
    simulationSpace=RootRelative
    alphaBoolBlend {blendOption=Linear}
    boundBone {boneName=dyn_front_hair_L_01}
    componentAppliedLinearAccClamp=<5000 5000 5000>
    componentLinearAccScale=<1 1 1>
    componentLinearVelScale=<1 1 1>
    gravityOverride {z=-980}
    physicsBodyDefinitions [{
      boundBone {boneName=dyn_front_hair_L_01}
      boxExtents=<20 20 20>
      constraintSetup {angularLimitsMax=<0 15 5>, angularLimitsMin=<0 -15 -15>, angularTarget=<1 0 0>}
      localJointOffset=<1 0 0>
     }]
    simSpaceSettings {maxAngularAcceleration=40, maxAngularVelocity=20, simSpaceAngularAlpha=1}
    sphericalLimits [{limitRadius=8, drivingBone {boneName=head}, sphereLocalOffset=<2 1 0>}]
   }
   n7 = Anim Dynamics - Bone: dyn_front_hair_L_02!(
    ComponentPose=@n6
    Alpha=1
    GravityOverride:Vector="0,0,-980"
    SimSpaceSettings:Anim Phys Sim Space Settings Structure=(SimSpaceAngularAlpha=1,MaxAngularVelocity=20,MaxAngularAcceleration=40,ExternalAngularVelocity=(X=0,Y=0,Z=0))
   ) -> (Pose) {
    angularDampingOverride=0.99
    angularSpringConstant=3000
    bAngularSpring=true
    bOverrideAngularDamping=true
    bUseGravityOverride=true
    bUseSphericalLimits=true
    simulationSpace=RootRelative
    alphaBoolBlend {blendOption=Linear}
    boundBone {boneName=dyn_front_hair_L_02}
    componentAppliedLinearAccClamp=<5000 5000 5000>
    componentLinearAccScale=<1 1 1>
    componentLinearVelScale=<1 1 1>
    gravityOverride {z=-980}
    physicsBodyDefinitions [{
      collisionType=CustomSphere
      sphereCollisionRadius=1.2
      boundBone {boneName=dyn_front_hair_L_02}
      boxExtents=<30 30 30>
      constraintSetup {angularLimitsMax=<0 10 10>, angularLimitsMin=<0 -10 -20>, angularTarget=<1 0 0>}
      localJointOffset=<2 0 0>
     }]
    simSpaceSettings {maxAngularAcceleration=40, maxAngularVelocity=20, simSpaceAngularAlpha=1}
    sphericalLimits [{limitRadius=8, drivingBone {boneName=head}, sphereLocalOffset=<0 1.5 0>}]
   }
   n8 = Anim Dynamics - Bone: dyn_front_hair_R_01!(
    ComponentPose=@n7
    Alpha=1
    GravityOverride="0,0,-980"
    SimSpaceSettings:Anim Phys Sim Space Settings Structure=(SimSpaceAngularAlpha=1,MaxAngularVelocity=20,MaxAngularAcceleration=40,ExternalAngularVelocity=(X=0,Y=0,Z=0))
   ) -> (Pose) {
    angularDampingOverride=1
    angularSpringConstant=2000
    bAngularSpring=true
    bOverrideAngularDamping=true
    bUseGravityOverride=true
    gravityScale=0.35
    simulationSpace=RootRelative
    alphaBoolBlend {blendOption=Linear}
    boundBone {boneName=dyn_front_hair_R_01}
    componentAppliedLinearAccClamp=<5000 5000 5000>
    componentLinearAccScale=<1 1 1>
    componentLinearVelScale=<1 1 1>
    gravityOverride {z=-980}
    physicsBodyDefinitions [{
      boundBone {boneName=dyn_front_hair_R_01}
      boxExtents=<20 20 20>
      constraintSetup {angularLimitsMax=<0 5 10>, angularLimitsMin=<0 -5 0>, angularTarget=<1 0 0>}
      localJointOffset=<1 0 0>
     }]
    simSpaceSettings {maxAngularAcceleration=40, maxAngularVelocity=20, simSpaceAngularAlpha=1}
   }
   n9 = Anim Dynamics - Bone: dyn_front_hair_R_02!(
    ComponentPose=@n8
    Alpha=1
    GravityOverride="0,0,-980"
    SimSpaceSettings:Anim Phys Sim Space Settings Structure=(SimSpaceAngularAlpha=1,MaxAngularVelocity=20,MaxAngularAcceleration=40,ExternalAngularVelocity=(X=0,Y=0,Z=0))
   ) -> (Pose) {
    angularDampingOverride=1
    angularSpringConstant=2000
    bAngularSpring=true
    bOverrideAngularDamping=true
    bUseGravityOverride=true
    gravityScale=0.35
    simulationSpace=RootRelative
    alphaBoolBlend {blendOption=Linear}
    boundBone {boneName=dyn_front_hair_R_02}
    componentAppliedLinearAccClamp=<5000 5000 5000>
    componentLinearAccScale=<1 1 1>
    componentLinearVelScale=<1 1 1>
    gravityOverride {z=-980}
    physicsBodyDefinitions [{
      boundBone {boneName=dyn_front_hair_R_02}
      boxExtents=<20 20 20>
      constraintSetup {angularLimitsMax=<0 5 10>, angularLimitsMin=<0 -15 0>, angularTarget=<1 0 0>}
      localJointOffset=<2 0 0>
     }]
    simSpaceSettings {maxAngularAcceleration=40, maxAngularVelocity=20, simSpaceAngularAlpha=1}
   }
   n10 = Anim Dynamics - Bone: dyn_front_hair_R_03!(
    ComponentPose=@n9
    Alpha=1
    GravityOverride="0,0,-980"
    SimSpaceSettings:Anim Phys Sim Space Settings Structure=(SimSpaceAngularAlpha=1,MaxAngularVelocity=20,MaxAngularAcceleration=40,ExternalAngularVelocity=(X=0,Y=0,Z=0))
   ) -> (Pose) {
    angularDampingOverride=1
    angularSpringConstant=2000
    bAngularSpring=true
    bOverrideAngularDamping=true
    bUseGravityOverride=true
    gravityScale=0.35
    simulationSpace=RootRelative
    alphaBoolBlend {blendOption=Linear}
    boundBone {boneName=dyn_front_hair_R_03}
    componentAppliedLinearAccClamp=<5000 5000 5000>
    componentLinearAccScale=<1 1 1>
    componentLinearVelScale=<1 1 1>
    gravityOverride {z=-980}
    physicsBodyDefinitions [{
      boundBone {boneName=dyn_front_hair_R_03}
      boxExtents=<20 20 20>
      constraintSetup {angularLimitsMax=<0 5 20>, angularLimitsMin=<0 -5 0>, angularTarget=<1 0 0>}
      localJointOffset=<2 0 0>
     }]
    simSpaceSettings {maxAngularAcceleration=40, maxAngularVelocity=20, simSpaceAngularAlpha=1}
   }
   n11 = Anim Dynamics - Bone: dyn_front_hair_A_02!(
    ComponentPose=@n10
    Alpha=1
    GravityOverride:Vector="0,0,-980"
    SimSpaceSettings:Anim Phys Sim Space Settings Structure=(SimSpaceAngularAlpha=1,MaxAngularVelocity=20,MaxAngularAcceleration=40,ExternalAngularVelocity=(X=0,Y=0,Z=0))
   ) -> (Pose) {
    angularDampingOverride=0.99
    angularSpringConstant=2000
    bAngularSpring=true
    bOverrideAngularDamping=true
    bUseGravityOverride=true
    bUseSphericalLimits=true
    simulationSpace=RootRelative
    alphaBoolBlend {blendOption=Linear}
    boundBone {boneName=dyn_front_hair_A_02}
    componentAppliedLinearAccClamp=<3000 3000 3000>
    componentLinearAccScale=<0.4 0.4 0.4>
    componentLinearVelScale=<0.6 0.6 0.6>
    gravityOverride {z=-980}
    physicsBodyDefinitions [{
      sphereCollisionRadius=2
      boundBone {boneName=dyn_front_hair_A_02}
      boxExtents=<20 20 20>
      constraintSetup {angularLimitsMax=<0 6 5>, angularLimitsMin=<0 -20 -15>, angularTarget=<1 0 0>}
      localJointOffset=<1 0 0>
     }]
    simSpaceSettings {maxAngularAcceleration=40, maxAngularVelocity=20, simSpaceAngularAlpha=1}
    sphericalLimits [{limitRadius=10, drivingBone {boneName=head}, sphereLocalOffset=<2 0 0>}]
   }
   n12 = Anim Dynamics - Bone: dyn_front_hair_B_02!(
    ComponentPose=@n11
    Alpha=1
    GravityOverride:Vector="0,0,-980"
    SimSpaceSettings:Anim Phys Sim Space Settings Structure=(SimSpaceAngularAlpha=1,MaxAngularVelocity=20,MaxAngularAcceleration=40,ExternalAngularVelocity=(X=0,Y=0,Z=0))
   ) -> (Pose) {
    angularDampingOverride=0.99
    angularSpringConstant=2000
    bAngularSpring=true
    bOverrideAngularDamping=true
    bUseGravityOverride=true
    bUseSphericalLimits=true
    simulationSpace=RootRelative
    alphaBoolBlend {blendOption=Linear}
    boundBone {boneName=dyn_front_hair_B_02}
    componentAppliedLinearAccClamp=<3000 3000 3000>
    componentLinearAccScale=<0.4 0.4 0.4>
    componentLinearVelScale=<0.6 0.6 0.6>
    gravityOverride {z=-980}
    physicsBodyDefinitions [{
      boundBone {boneName=dyn_front_hair_B_02}
      boxExtents=<20 20 20>
      constraintSetup {angularLimitsMax=<0 5 20>, angularLimitsMin=<0 -5 -20>, angularTarget=<1 0 0>}
      localJointOffset=<1 0 0>
     }]
    simSpaceSettings {maxAngularAcceleration=40, maxAngularVelocity=20, simSpaceAngularAlpha=1}
    sphericalLimits [{limitRadius=10, drivingBone {boneName=head}, sphereLocalOffset=<2 0 0>}]
   }
   n13 = RigidBody!(ComponentPose=@n12, Alpha, OverrideWorldGravity:Vector="0,0,-980", ExternalForce) -> (Pose) {
    bOverrideWorldGravity=true
    overridePhysicsAsset=&/F_MED_RedJay_FaceAcc_earrings_Physics
    simulationSpace=BaseBoneSpace
    alphaBoolBlend {blendOption=Linear}
    baseBoneRef {boneName=head}
    componentAppliedLinearAccClamp=<2000 2000 2000>
    overrideWorldGravity {z=-980}
    simSpaceSettings {
     dampingAlpha=0
     maxAngularAcceleration=20
     maxAngularVelocity=10
     maxLinearAcceleration=1400
     maxLinearVelocity=700
     worldAlpha=1
     externalLinearDragV=<0.2 0.2 0.2>
    }
   }
   Save cached pose 'CRCache_dynamics'(Pose=Component To Local(ComponentPose=@n13)) {cacheName=CRCache_dynamics}
  ]}
 EventGraph {nodes [
   Event BlueprintUpdateAnimation()
   > n1 = Sequence()
   > n2 = Gravity Override State Function DEPRECATED(CharacterPartAnimInstance=Self-Reference(), GameTimeInSeconds=GetGameTimeInSeconds(), GravityOverrideParams=Get Gravity Override Params())
   > Set GravityOverride(GravityOverride=@n2)
   n3 = Branch(execute=@n1.then_1, Condition=Get bIsInFrontEnd())
   > Set SimSpace AnimDynamics(SimSpace AnimDynamics:Anim Phys Sim Space Settings Structure=Get SimSpace AnimDynamics_frondEnd() -> (SimSpace AnimDynamics_frondEnd:Anim Phys Sim Space Settings Structure)) -> (Output_Get:Anim Phys Sim Space Settings Structure)
   Set SimSpace AnimDynamics(
    execute=@n3.else
    SimSpace AnimDynamics=Get SimSpace AnimDynamics_default() -> (SimSpace AnimDynamics_default:Anim Phys Sim Space Settings Structure)
   )
  ]}
]
variables [
 Gravity Override Params {type=GravityOverrideParamsStruct}
 GravityOverride {type=Vector}
 SimSpace AnimDynamics {
  type=AnimPhysSimSpaceSettings
  default {maxAngularAcceleration=40, maxAngularVelocity=20, simSpaceAngularAlpha=0.8, externalAngularVelocity=<0 0 0>}
 }
 SimSpace AnimDynamics_default {
  type=AnimPhysSimSpaceSettings
  default {maxAngularAcceleration=40, maxAngularVelocity=20, simSpaceAngularAlpha=1, externalAngularVelocity=<0 0 0>}
 }
 SimSpace AnimDynamics_frondEnd {
  type=AnimPhysSimSpaceSettings
  default {maxAngularAcceleration=40, maxAngularVelocity=20, simSpaceAngularAlpha=0.8, externalAngularVelocity=<0 0 0>}
 }
]
